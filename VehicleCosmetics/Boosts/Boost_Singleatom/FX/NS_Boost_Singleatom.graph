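# NS_Boost_Singleatom is the Niagara effect for the single-port Singleatom vehicle boost cosmetic, anchored behind DTCarOffset plus SingleBoostPos.
# Near-identical CPU and GPU sprite emitters stream backward, drawn toward a line by attraction.
# A persistent burst builds rotated meshes and orbiting sprites, joined by spawn-group glows, point-attracted sparks and swirling motes.
# Every emitter takes painted-variant colors from VC_PaintedVariantsOverride, and spawn-group visibility tags separate renderers.
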NiagaraSystem /VehicleCosmetics/Boosts/Boost_Singleatom/FX/NS_Boost_Singleatom
dependencies {
 usedDataInterfaces [
  &NiagaraDataInterfaceArrayColor
  &NiagaraDataInterfaceCurve
  &NiagaraDataInterfaceMeshRendererInfo
  &NiagaraDataInterfaceArrayFloat
  /Script/Niagara.NiagaraDataInterfaceSimpleCounter
  &NiagaraDataInterfaceArrayInt32
  &NiagaraDataInterfaceArrayFloat3
  &NiagaraDataInterfaceSkeletalMesh
 ]
 usedDynamicInputs []
 usedModules [
  &self:SystemSpawnScript.SystemScriptSource.SystemScriptGraph.NiagaraNodeAssignment_0.SetVariables_#1
  &self:SystemSpawnScript.SystemScriptSource.SystemScriptGraph.NiagaraNodeAssignment_1.SetVariables_#2
  &self:SystemSpawnScript.SystemScriptSource.SystemScriptGraph.NiagaraNodeAssignment_2.SetVariables_#3
  &SystemState
 ]
}
emitters {
 NE_VC_SingleBoost_Sprite_Parent {
  emitterUpdateScript [
   &EmitterState()
   &SpawnRate(SpawnRate=70, Use Spawn Probability=false)
  ]
  particleSpawnScript [
   &VC_PaintedVariantsOverride(
    Black=&SelectLinearColorFromArray(
     Array Sampling Mode=::NewEnumerator0
     Direct Array Index=?
     Color Selection Array=DataInterface() {propertyValues {ColorData [<rgba 0 0 0 1.2>, <rgba 0.15 0.15 0.15 1.2>]}}
     Random Seed=0
    )
    BurntSienna=&RandomRangeLinearColor(Minimum=<rgba 0.6724 0.2346 0 1.2>, Maximum=<rgba 0.2 0.0394 0.0052 1.2>)
    Cobalt=&RandomRangeLinearColor(Minimum=<rgba 0 0.0127 0.6841 3>, Maximum=<rgba 0.52 0.5252 0.8 3>)
    Crimson=&RandomRangeLinearColor(Minimum=<rgba 0.4635 0.0244 0.0093 1.2>, Maximum=<rgba 1 0.03 0.03 1.2>)
    DefaultColor=&RandomRangeLinearColor(Minimum=<rgba 1 0.7504 0.18 1>, Maximum=<rgba 1 0.8733 0.05 1>)
    ForestGreen=&RandomRangeLinearColor(Minimum=<rgba 0 1 0.05 1.5>, Maximum=<rgba 0.1125 0.45 0.1294 2>)
    Grey=<rgba 0 0 0 1.5>
    Lime=&RandomRangeLinearColor(Minimum=<rgba 0.3333 0.5 0 1.2>, Maximum=<rgba 1 1 0 1.2>)
    Pink=&RandomRangeLinearColor(Minimum=<rgba 1 0.4 0.79 2>, Maximum=<rgba 1 0.6 0.86 2>)
    Purple=&RandomRangeLinearColor(Minimum=<rgba 0.1231 0 0.3 1.2>, Maximum=<rgba 0.3 0.1 0.5 1.2>)
    Saffron=<rgba 1.2 0.92 0 1>
    SkyBlue=&RandomRangeLinearColor(Minimum=<rgba 0 0.5333 0.8 1.2>, Maximum=<rgba 0.5 0.8333 1 1.2>)
    TitaniumWhite=<rgba 2 2 2 1.5>
   )
   &InitializeParticle(
    Color Mode=::NewEnumerator0
    Lifetime=&RandomRangeFloat(Minimum=0.5, Maximum=0.65)
    Position Offset=&AddVectorToPosition(
     Position=<-85 0 0>
     Vector=&AddVectorToPosition(Position=$User.DTCarOffset, Vector=Linked() {linkedVariable {name=User.SingleBoostPos, type=NiagaraPosition}})
    )
    Position Offset Coordinate Space=::Local
    Sprite Rotation Mode=::NewEnumerator2
    Sprite Size Mode=::NewEnumerator1
    Uniform Sprite Size Max=8
    Uniform Sprite Size Min=3
    UsePositionOffset=true
    Write Material Random=true
   )
   &ShapeLocation(MaskSpawnGroup=false, Override Seed=false, Sphere Radius=64, Sphere Surface Distribution=1, UseOwnerScale=true)
   &AddVelocity(Velocity=<-800 0 0>)
  ]
  particleUpdateScript [
   &ParticleState()
   &ScaleSpriteSize(Uniform Curve Sprite Scale=DataInterface() {propertyValues {Curve {keys [{tangentMode=RCTM_User, time=0.2, value=1}, {tangentMode=RCTM_User, time=1, value=0.7}]}}})
   /Niagara/Modules/Update/Forces/LineAttractionForce(
    Attraction Strength=190
    Line End:Vector3f=&AddVectorToPosition(
     Position=<-700 0 0>
     Vector=&AddVectorToPosition(Position=$User.DTCarOffset, Vector=Linked() {linkedVariable {name=User.SingleBoostPos, type=NiagaraPosition}})
    )
    Line Segment in Localspace=true
    Line Start:Vector3f=&AddVectorToPosition(
     Position=<-85 0 0>
     Vector=&AddVectorToPosition(Position=$User.DTCarOffset, Vector=Linked() {linkedVariable {name=User.SingleBoostPos, type=NiagaraPosition}})
    )
    Use Attraction Falloff=false
   )
   &InheritSourceMovement(ApplyForce=false, ApplyPosition=true, ApplyPositionFalloff=false)
   &SolveForcesAndVelocity(Clamp Velocity=false, Limit Acceleration=false)
  ]
  rendererClasses [&NiagaraSpriteRendererProperties]
  renderers [{rendererClass=&NiagaraSpriteRendererProperties}]
 }
 NE_VC_SingleBoost_Sprite_Parent002 {
  simTarget=GPUComputeSim
  emitterUpdateScript [
   &EmitterState()
   &SpawnRate(SpawnRate=105, Use Spawn Probability=false)
  ]
  particleSpawnScript [
   &VC_PaintedVariantsOverride(
    Black=&SelectLinearColorFromArray(
     Array Sampling Mode=::NewEnumerator0
     Direct Array Index=?
     Color Selection Array=DataInterface() {propertyValues {ColorData [<rgba 0 0 0 1.2>, <rgba 0.15 0.15 0.15 1.2>]}}
     Random Seed=0
    )
    BurntSienna=&RandomRangeLinearColor(Minimum=<rgba 0.6724 0.2346 0 1.2>, Maximum=<rgba 0.2 0.0394 0.0052 1.2>)
    Cobalt=&RandomRangeLinearColor(Minimum=<rgba 0 0.0127 0.6841 3>, Maximum=<rgba 0.52 0.5252 0.8 3>)
    Crimson=&RandomRangeLinearColor(Minimum=<rgba 0.4635 0.0244 0.0093 1.2>, Maximum=<rgba 1 0.03 0.03 1.2>)
    DefaultColor=&RandomRangeLinearColor(Minimum=<rgba 1 0.7504 0.18 1>, Maximum=<rgba 1 0.8733 0.05 1>)
    ForestGreen=&RandomRangeLinearColor(Minimum=<rgba 0 1 0.05 1.5>, Maximum=<rgba 0.1125 0.45 0.1294 2>)
    Grey=<rgba 0 0 0 1.5>
    Lime=&RandomRangeLinearColor(Minimum=<rgba 0.3333 0.5 0 1.2>, Maximum=<rgba 1 1 0 1.2>)
    Pink=&RandomRangeLinearColor(Minimum=<rgba 1 0.4 0.79 2>, Maximum=<rgba 1 0.6 0.86 2>)
    Purple=&RandomRangeLinearColor(Minimum=<rgba 0.1231 0 0.3 1.2>, Maximum=<rgba 0.3 0.1 0.5 1.2>)
    Saffron=<rgba 1.2 0.92 0 1>
    SkyBlue=&RandomRangeLinearColor(Minimum=<rgba 0 0.5333 0.8 1.2>, Maximum=<rgba 0.5 0.8333 1 1.2>)
    TitaniumWhite=<rgba 2 2 2 1.5>
   )
   &InitializeParticle(
    Color Mode=::NewEnumerator0
    Lifetime=&RandomRangeFloat(Minimum=0.5, Maximum=0.65)
    Position Offset=&AddVectorToPosition(
     Position=<-85 0 0>
     Vector=&AddVectorToPosition(Position=$User.DTCarOffset, Vector=Linked() {linkedVariable {name=User.SingleBoostPos, type=NiagaraPosition}})
    )
    Position Offset Coordinate Space=::Local
    Sprite Rotation Mode=::NewEnumerator2
    Sprite Size Mode=::NewEnumerator1
    Uniform Sprite Size Max=8
    Uniform Sprite Size Min=3
    UsePositionOffset=true
    Write Material Random=true
   )
   &ShapeLocation(MaskSpawnGroup=false, Override Seed=false, Sphere Radius=64, Sphere Surface Distribution=1, UseOwnerScale=true)
   &AddVelocity(Velocity=<-800 0 0>)
  ]
  particleUpdateScript [
   &ParticleState()
   &ScaleSpriteSize(Uniform Curve Sprite Scale=DataInterface() {propertyValues {Curve {keys [{tangentMode=RCTM_User, time=0.2, value=1}, {tangentMode=RCTM_User, time=1, value=0.7}]}}})
   &LineAttractionForce(
    Attraction Strength=190
    Line End:Vector3f=&AddVectorToPosition(
     Position=<-700 0 0>
     Vector=&AddVectorToPosition(Position=$User.DTCarOffset, Vector=Linked() {linkedVariable {name=User.SingleBoostPos, type=NiagaraPosition}})
    )
    Line Segment in Localspace=true
    Line Start:Vector3f=&AddVectorToPosition(
     Position=<-85 0 0>
     Vector=&AddVectorToPosition(Position=$User.DTCarOffset, Vector=Linked() {linkedVariable {name=User.SingleBoostPos, type=NiagaraPosition}})
    )
    Use Attraction Falloff=false
   )
   &InheritSourceMovement(ApplyForce=false, ApplyPosition=true, ApplyPositionFalloff=false)
   &SolveForcesAndVelocity(Clamp Velocity=false, Limit Acceleration=false)
  ]
  rendererClasses [&NiagaraSpriteRendererProperties]
  renderers [{rendererClass=&NiagaraSpriteRendererProperties}]
 }
 NE_VC_SingleBoost_Sprite_Parent007 {
  emitterUpdateScript [
   &EmitterState(
    Inactive Response=::NewEnumerator1
    Life Cycle Mode=::NewEnumerator1
    Loop Behavior=::NewEnumerator1
    Loop Duration Mode=::NewEnumerator1
    UseLoopDelay=false
   )
   &SpawnBurst_Instantaneous(Spawn Count=8, Use Spawn Probability=false, UseLoopCountLimit=false) {module=SpawnBurst_Instantaneous001}
   &SpawnBurst_Instantaneous(Spawn Count=26, Spawn Group=1, Use Spawn Probability=false, UseLoopCountLimit=false)
   &SpawnRate(SpawnGroup=1, SpawnRate=8, Use Spawn Probability=false) {enabled=false}
  ]
  particleSpawnScript [
   &VC_PaintedVariantsOverride(
    Black=<rgba 0 0 0 2>
    BurntSienna=<rgba 0.75 0.2616 0 1>
    Cobalt=<rgba 0.18 0.1989 1.2 1>
    Crimson=<rgba 0.45 0.009 0.009 1.2>
    DefaultColor=<rgba 1.5 0.5215 0.36 1>
    ForestGreen=<rgba 0.025 0.5 0.049 1.2>
    Grey=<rgba 0.4 0.4 0.4 2>
    Lime=<rgba 0.6667 1 0 1.5>
    Pink=<rgba 1 0.44 0.7964 2>
    Purple=<rgba 0.2462 0 0.6 1>
    R1 Black:LinearColor=<rgba 0 0 0 2>
    R1 BurntSienna:LinearColor=<rgba 0.6724 0.2346 0 1>
    R1 Cobalt:LinearColor=<rgba 0.35 0.362 1 1>
    R1 Crimson:LinearColor=<rgba 1 0.03 0.03 1>
    R1 DefaultColor=<rgba 2 1.4922 0.8 1>
    R1 ForestGreen:LinearColor=<rgba 0.12 1 0.1645 1>
    R1 Gold:LinearColor=<rgba 10 6.1917 1 1>
    R1 Grey:LinearColor=<rgba 0.8 0.8 0.8 1.2>
    R1 Lime:LinearColor=<rgba 1 1.5 0 1>
    R1 Pink:LinearColor=<rgba 1 0.7 0.8909 1>
    R1 Purple:LinearColor=<rgba 0.3283 0 0.8 1>
    R1 Saffron:LinearColor=<rgba 0.9 0.7215 0.135 1>
    R1 SkyBlue:LinearColor=<rgba 0.2 0.7323 1 1>
    R1 TitaniumWhite:LinearColor=<rgba 1 1 1 1>
    Saffron=<rgba 1 0.8017 0.15 1>
    SkyBlue=<rgba 0 0.6654 1 1>
    TitaniumWhite=<rgba 1 1 1 1>
    Use R1 Painted Overrides=true
   )
   &InitializeParticle(
    Color Mode=::NewEnumerator0
    Lifetime=0.2
    Mesh Renderer Array Visibility Mode=::NewEnumerator1
    Mesh Renderer Info=DataInterface() {propertyValues {MeshRenderer=&self:NE_VC_SingleBoost_Sprite_Parent007.NiagaraMeshRendererProperties_1}}
    Mesh Renderer Mesh Index:NiagaraInt32=&ReturnExecIndex()
    Mesh Scale=&Multiply_VectorByFloat(Vector=<1 0.22 1>, Float=0.58)
    Mesh Scale Mode=::NewEnumerator4
    Position Offset=&AddVectorToPosition(
     Position=<-85 0 0>
     Vector=&AddVectorToPosition(Position=$User.DTCarOffset, Vector=Linked() {linkedVariable {name=User.SingleBoostPos, type=NiagaraPosition}})
    )
    Position Offset Coordinate Space=::Local
    Sprite Size Mode=::NewEnumerator3
    Uniform Sprite Size=6.7
    UsePositionOffset=true
    Write Material Random=true
   )
   &AddVelocity(Velocity=<-1500 0 0>) {enabled=false}
   &InitialMeshOrientation(
    Mesh Orientation Mode=::NewEnumerator2
    Rotation=&MergeHelper_VectorSwitch(0=<0 0 0>, 1=<0.25 0 0>, 2=<0.125 0 0>, 3=<-0.125 0 0>, Index=&ReturnExecIndex(), Num=::NewEnumerator2)
    Use Rotation Vector=true
   )
   /SetVariables(Particles.VisibilityTag=$Emitter.SpawnGroup) {isSetParameters=true}
  ]
  particleUpdateScript [
   &ParticleState(Kill Particles When Lifetime Has Elapsed=false, Let Infinitely Lived Particles Die When Emitter Deactivates=true)
   &SolveForcesAndVelocity(Clamp Velocity=false, Limit Acceleration=false)
   &RotateAroundPoint(
    Radius:NiagaraFloat=&MergeHelper_FloatSwitch(0=0, 1=58, 2=?, 3=?, Index=&SpawnGroup(), Num=::NewEnumerator0)
    Rotation Center:NiagaraPosition=&AddVectorToPosition(
     Position=/Niagara/DynamicInputs/Helpers/SimulationPosition()
     Vector=&AddVectorToPosition(
      Position=<-85 0 0>
      Vector=&AddVectorToPosition(Position=$User.DTCarOffset, Vector=Linked() {linkedVariable {name=User.SingleBoostPos, type=NiagaraPosition}})
     )
    )
    Rotation Phase:NiagaraFloat=&MergeHelper_FloatSwitch(
     0=0
     1=&SelectFloatFromArray(
      Array Sampling Mode=::NewEnumerator0
      Direct Array Index=?
      Float Selection Array=DataInterface() {propertyValues {FloatData [0.72,0.98]}}
      Random Seed=0
     )
     2=?
     3=?
     Index=&SpawnGroup()
     Num=::NewEnumerator0
    )
    Rotation Rate:NiagaraFloat=&MergeHelper_FloatSwitch(0=0, 1=1.5, 2=?, 3=?, Index=&SpawnGroup(), Num=::NewEnumerator0)
    Y Axis:Vector3f=&MergeHelper_VectorSwitch(
     0=<0 -1 0>
     1=<0 -1 1>
     2=<0 1 1>
     3=<0 0 1>
     Index=&SelectIntFromArray(
      Direct Array Index=/Niagara/DynamicInputs/DataInterface/GenerateIntFromCounter(SimpleCounter:NiagaraDataInterfaceSimpleCounter=DataInterface() {propertyValues {GpuSyncMode=None, InitialValue=0}})
      Direct Array Mode=::NewEnumerator1
      Int Selection Array=DataInterface() {propertyValues {IntData [0,1,2,3]}}
     )
     Num=::NewEnumerator2
    )
   )
   &UpdateMeshOrientation(Rotation Rate=-1.5, Rotation Vector=<0 1 0>)
  ]
  rendererClasses [&NiagaraSpriteRendererProperties, &NiagaraMeshRendererProperties]
  renderers [
   {rendererClass=&NiagaraSpriteRendererProperties}
   {rendererClass=&NiagaraMeshRendererProperties, rendererIndex=1}
   {rendererClass=&NiagaraMeshRendererProperties, rendererIndex=2}
  ]
 }
 NE_VC_SingleBoost_Sprite_Parent008 {
  emitterUpdateScript [
   &EmitterState()
   &SpawnRate(Spawn Probability=0.6, SpawnRate=4, Use Spawn Probability=false)
   &SpawnRate(Spawn Probability=0.5, SpawnGroup=1, SpawnRate=15, Use Spawn Probability=true) {module=SpawnRate001}
   &SpawnRate(Spawn Probability=0.6, SpawnGroup=2, SpawnRate=25, Use Spawn Probability=false) {module=SpawnRate002}
  ]
  particleSpawnScript [
   /SetVariables(Particles.VisibilityTag=$Emitter.SpawnGroup) {isSetParameters=true}
   &VC_PaintedVariantsOverride(
    Black=<rgba 0 0 0 0.93>
    BurntSienna=<rgba 1.3 0.4333 0 0.6>
    Cobalt=<rgba 0.5 0.5833 5 1>
    Crimson=<rgba 2 0 0.0542 0.8>
    DefaultColor=<rgba 6 2.2882 0.3 1>
    ForestGreen=<rgba 0 2 0.1011 0.8>
    Grey=<rgba 1 1 1 1>
    Lime=<rgba 1 2 0 1>
    Pink=<rgba 3 1.2 2.3455 1>
    Purple=<rgba 3 0.3 1.9435 0.4>
    R1 Black:LinearColor=&SelectLinearColorFromArray(
     Array Sampling Mode=::NewEnumerator0
     Direct Array Index=?
     Color Selection Array=DataInterface() {propertyValues {ColorData [<rgba 0 0 0 1.2>, <rgba 0.15 0.15 0.15 1.2>]}}
     Random Seed=0
    )
    R1 BurntSienna:LinearColor=&SelectLinearColorFromArray(
     Array Sampling Mode=::NewEnumerator0
     Direct Array Index=?
     Color Selection Array=DataInterface() {propertyValues {ColorData [<rgba 0.6724 0.2346 0 1.2>, <rgba 0.5 0.0985 0.013 1.2>]}}
     Random Seed=0
    )
    R1 Cobalt:LinearColor=&SelectLinearColorFromArray(
     Array Sampling Mode=::NewEnumerator0
     Direct Array Index=?
     Color Selection Array=DataInterface() {propertyValues {ColorData [<rgba 0.18 0.1989 1.2 1.2>, <rgba 0.7 0.7056 1 1.2>]}}
     Random Seed=0
    )
    R1 Crimson:LinearColor=&SelectLinearColorFromArray(
     Array Sampling Mode=::NewEnumerator0
     Direct Array Index=?
     Color Selection Array=DataInterface() {propertyValues {ColorData [<rgba 1 0 0 1.2>, <rgba 1 0.12 0.12 1.2>]}}
     Random Seed=0
    )
    R1 DefaultColor=&RandomRangeLinearColor(Minimum=<rgba 1 0.3625 0.15 2>, Maximum=<rgba 1 0.6733 0.02 2>)
    R1 ForestGreen:LinearColor=&SelectLinearColorFromArray(
     Array Sampling Mode=::NewEnumerator0
     Direct Array Index=?
     Color Selection Array=DataInterface() {propertyValues {ColorData [<rgba 0 1 0.05 1.5>, <rgba 0.06 0.2 0.067 1.5>]}}
     Random Seed=0
    )
    R1 Grey:LinearColor=<rgba 0 0 0 1>
    R1 Lime:LinearColor=<rgba 0.2943 0.5885 0 2>
    R1 Pink:LinearColor=&SelectLinearColorFromArray(
     Array Sampling Mode=::NewEnumerator0
     Direct Array Index=?
     Color Selection Array=DataInterface() {propertyValues {ColorData [<rgba 1.2 0.3 0.885 2>, <rgba 1 0.6 0.86 2>]}}
     Random Seed=0
    )
    R1 Purple:LinearColor=&SelectLinearColorFromArray(
     Array Sampling Mode=::NewEnumerator0
     Direct Array Index=&RandomRangeInt(Minimum=0, Maximum=1)
     Color Selection Array=DataInterface() {propertyValues {ColorData [<rgba 0.4104 0 1 1>, <rgba 0.7922 0.12 0.8 1>]}}
     Random Seed=0
    )
    R1 Saffron:LinearColor=<rgba 1 0.7667 0 1>
    R1 SkyBlue:LinearColor=&SelectLinearColorFromArray(
     Array Sampling Mode=::NewEnumerator0
     Direct Array Index=?
     Color Selection Array=DataInterface() {propertyValues {ColorData [<rgba 0 1.3333 2 1.2>, <rgba 0.7 0.9 1 1.2>]}}
     Random Seed=0
    )
    R1 TitaniumWhite:LinearColor=<rgba 2 2 2 1>
    Saffron=<rgba 2 1.5333 0 1>
    SkyBlue=<rgba 0.3 2.0965 3 0.5>
    TitaniumWhite=<rgba 2 2 2 1>
    Use R1 Painted Overrides=true
   )
   &InitializeParticle(
    Color=$Particles.Color
    Lifetime=&MergeHelper_FloatSwitch(
     0=0.35
     1=0.3
     2=&RandomRangeFloat(Minimum=0.8, Maximum=1.5)
     3=?
     Index=$Emitter.SpawnGroup
     Num=::NewEnumerator1
    )
    Position Offset=&AddVectorToPosition(
     Position=<-85 0 0>
     Vector=&AddVectorToPosition(Position=$User.DTCarOffset, Vector=Linked() {linkedVariable {name=User.SingleBoostPos, type=NiagaraPosition}})
    )
    Position Offset Coordinate Space=::Local
    Sprite Rotation Mode=::NewEnumerator2
    Sprite Size Mode=::NewEnumerator3
    Uniform Sprite Size=42
    UsePositionOffset=true
    Write Material Random=true
   )
   &AddVelocity(
    Module.Use Falloff Distance=false
    Override Seed=false
    Velocity Mode=::NewEnumerator1
    Velocity Origin=$StackContext.Position
    Velocity Speed=&MergeHelper_FloatSwitch(
     0=25
     1=85
     2=&RandomRangeFloat(Minimum=215, Maximum=285)
     3=?
     Index=&SpawnGroup()
     Num=::NewEnumerator1
    )
   )
  ]
  particleUpdateScript [
   &ParticleState()
   &PointAttractionForce(
    Attraction Radius:NiagaraFloat=&MergeHelper_FloatSwitch(0=0, 1=0, 2=110, 3=?, Index=&SpawnGroup(), Num=::NewEnumerator1)
    Attraction Strength=&MergeHelper_FloatSwitch(
     0=0
     1=0
     2=&RandomRangeFloat(Minimum=60, Maximum=100)
     3=?
     Index=&SpawnGroup()
     Num=::NewEnumerator1
    )
    Attractor Position:NiagaraPosition=&AddVectorToPosition(
     Position=$Engine.Emitter.SimulationPosition
     Vector=&AddVectorToPosition(
      Position=<-85 0 0>
      Vector=&AddVectorToPosition(Position=$User.DTCarOffset, Vector=Linked() {linkedVariable {name=User.SingleBoostPos, type=NiagaraPosition}})
     )
    )
    Falloff Exponent:NiagaraFloat=&MergeHelper_FloatSwitch(0=0, 1=0, 2=0.48, 3=?, Index=&SpawnGroup(), Num=::NewEnumerator1)
    Kill Within Radius=false
    Module.Use Falloff=true
    UseAttractorOffset=false
   )
   &SolveForcesAndVelocity(Clamp Velocity=false, Limit Acceleration=false)
   &ScaleSpriteSize(
    Uniform Curve Scale=1.5
    Uniform Curve Sprite Scale=DataInterface() {propertyValues {Curve {
       defaultValue=0
       keys [
        {interpMode=RCIM_Cubic, tangentMode=RCTM_User, time=0, value=0}
        {arriveTangent=-0.0184, leaveTangent=-0.0184, time=0.2, value=1}
        {arriveTangent=-2.1486, tangentMode=RCTM_User, time=1, value=0}
       ]
      }}}
   )
   /SetVariables(
    Particles.BurstSize:Vector2f=&Multiply_Vector2D_ByFloat(Float=&RandomRangeFloat(Minimum=4, Maximum=7), Vector2D=<1 1>)
    Particles.SparksColor:LinearColor=&HueShiftLinearColor(
     Hue=0
     Saturation=1
     Value=&FloatFromCurve(FloatCurve=DataInterface() {propertyValues {Curve {keys [<at 0.3 1>,<at 1 0.3>]}}})
     Linear Color=$Particles.R1 Color
    )
    Particles.SparksSize:Vector2f=&Multiply_Vector2D_ByFloat(
     Float=&FloatFromCurve(FloatCurve=DataInterface() {propertyValues {Curve {keys [{tangentMode=RCTM_User, time=0, value=0}, {tangentMode=RCTM_User, time=1, value=1}]}}})
     Vector2D=&RandomRangeVector2D(Minimum=<6 40>, Maximum=<9 55>)
    )
   ) {isSetParameters=true}
  ]
  rendererClasses [&NiagaraSpriteRendererProperties]
  renderers [
   {rendererClass=&NiagaraSpriteRendererProperties}
   {rendererClass=&NiagaraSpriteRendererProperties, rendererIndex=1}
   {rendererClass=&NiagaraSpriteRendererProperties, rendererIndex=2}
  ]
 }
 NE_VC_SingleBoost_Sprite_Parent009 {
  emitterUpdateScript [
   &EmitterState()
   &SpawnRate(SpawnRate=40, Use Spawn Probability=false)
  ]
  particleSpawnScript [
   &VC_PaintedVariantsOverride(
    Black=<rgba 0 0 0 1.2>
    BurntSienna=<rgba 0.6724 0.2346 0 1.2>
    Cobalt=<rgba 0 0.037 2 3>
    Crimson=<rgba 0.8 0 0 1.2>
    DefaultColor=&RandomRangeLinearColor(Minimum=<rgba 1 0.2783 0 1>, Maximum=<rgba 2 0.8333 0 1>)
    ForestGreen=<rgba 0 0.8 0.0404 1.5>
    Grey=<rgba 0 0 0 1.5>
    Lime=<rgba 1.25 1.5 0 2>
    Pink=<rgba 1 0.3 0.7455 2>
    Purple=<rgba 0.4281 0.03 1 2>
    Saffron=<rgba 2 1.5333 0 1>
    SkyBlue=<rgba 0 1.3308 2 1.2>
    TitaniumWhite=<rgba 2 2 2 1.2>
   )
   &InitializeParticle(
    Color Mode=::NewEnumerator0
    Lifetime=0.6
    Position Offset=&AddVectorToPosition(
     Position=<-85 0 0>
     Vector=&AddVectorToPosition(Position=$User.DTCarOffset, Vector=Linked() {linkedVariable {name=User.SingleBoostPos, type=NiagaraPosition}})
    )
    Position Offset Coordinate Space=::Local
    Sprite Size Mode=::NewEnumerator3
    Uniform Sprite Size=&RandomRangeFloat(Minimum=2.5, Maximum=5)
    UsePositionOffset=true
    Write Material Random=true
   )
   &AddVelocity(Velocity=<-1500 0 0>) {enabled=false}
  ]
  particleUpdateScript [
   &ParticleState()
   &SolveForcesAndVelocity(Clamp Velocity=false, Limit Acceleration=false)
   &RotateAroundPoint(
    Radius:NiagaraFloat=&SelectFloatFromArray(
     Array Sampling Mode=::NewEnumerator0
     Direct Array Index=?
     Float Selection Array=DataInterface() {propertyValues {FloatData [5,2,6,3]}}
     Random Seed=0
    )
    Rotation Center:NiagaraPosition=&AddVectorToPosition(
     Position=&SimulationPosition()
     Vector=&AddVectorToPosition(
      Position=<-85 0 0>
      Vector=&AddVectorToPosition(Position=$User.DTCarOffset, Vector=Linked() {linkedVariable {name=User.SingleBoostPos, type=NiagaraPosition}})
     )
    )
    Rotation Phase:NiagaraFloat=&RandomRangeFloat(Minimum=0, Maximum=1)
    Rotation Rate:NiagaraFloat=&SelectFloatFromArray(
     Array Sampling Mode=::NewEnumerator0
     Direct Array Index=?
     Float Selection Array=DataInterface() {propertyValues {FloatData [-2,-3,5,2.5]}}
     Random Seed=0
    )
    Y Axis:Vector3f=&SelectVectorFromArray(
     Array Sampling Mode=::NewEnumerator0
     Direct Array Index=?
     Vector Selection Array=DataInterface() {propertyValues {FloatData [<0 1 0>, <0 -1 1>, <0 1 1>, <0 0 1>]}}
     Random Seed=0
    )
   )
   &ShapeLocation(
    MaskSpawnGroup=false
    Override Seed=false
    Shape Origin=$Output.RotateAroundPoint.OutputPosition
    Sphere Radius=73
    Sphere Surface Distribution=0.8
    UseOwnerScale=true
   )
   &ScaleColor(
    Scale RGB=&Multiply_VectorByFloat(
     Vector=<1 1 1>
     Float=&FloatFromCurve(FloatCurve=DataInterface() {propertyValues {Curve {keys [<at 0 1>,<at 1 0.3>]}}})
    )
    ScaleA=true
    ScaleRGB=true
   )
  ]
  rendererClasses [&NiagaraSpriteRendererProperties]
  renderers [{rendererClass=&NiagaraSpriteRendererProperties}]
 }
}
summary {
 systemName=NS_Boost_Singleatom
 emitters [
  {emitterName=NE_VC_SingleBoost_Sprite_Parent002, simTarget=GPUComputeSim, rendererClasses [&NiagaraSpriteRendererProperties]}
  {
   emitterName=NE_VC_SingleBoost_Sprite_Parent007
   rendererClasses [&NiagaraSpriteRendererProperties, &NiagaraMeshRendererProperties]
  }
  {emitterName=NE_VC_SingleBoost_Sprite_Parent008, rendererClasses [&NiagaraSpriteRendererProperties]}
  {emitterName=NE_VC_SingleBoost_Sprite_Parent009, rendererClasses [&NiagaraSpriteRendererProperties]}
  {emitterName=NE_VC_SingleBoost_Sprite_Parent, rendererClasses [&NiagaraSpriteRendererProperties]}
 ]
 userVariables [
  User.bIsFrontendPreview {type=NiagaraBool}
  User.bIsPainted {type=NiagaraBool}
  User.bPreviewVehicleSpeed {type=NiagaraBool}
  User.ColorVariantID {type=NiagaraInt32}
  User.DTCarOffset {type=NiagaraPosition, description="", defaultValue {}}
  User.GameModeID {type=NiagaraInt32}
  User.PaintedColor {type=LinearColor, description="", defaultValue {struct=LinearColor, value=<rgba 1 1 1 1>}}
  User.PaintedScalar {type=NiagaraFloat, defaultValue=1, description=""}
  User.PortCount {type=NiagaraInt32, defaultValue=1}
  User.SingleBoostPos {type=NiagaraPosition, description="", defaultValue {}}
 ]
}
